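; Base team-healing pizza box: a thrown building actor holding limited slices that players eat through a consume GameplayEffect or take as items.
; Balance rows drive slice count, lifespan, interact times and health/shield eligibility.
; RepNotifies animate the lid and hide missing slices through material parameters.
; It marks itself for the squad, handles vehicle and water landings, and tosses out a percentage of leftover slices on KillSelf-tagged death.
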(event Custom|OpenBox
  ;[1 1 1] Open Box
  (bind (_direction _paramupdate) (Timeline|OpenBoxTimeline Play)
    (:Update
      (Rendering|Material|SetScalarParameterValueonMaterials (Variables|Default|GetStaticMesh) (Variables|Default|GetMatParamName_BoxOpen) _paramupdate))))

(event Custom|CloseBox
  ;[1 1 1] Open Box
  (Timeline|OpenBoxTimeline Reverse))

(event EventBeginPlay
  ;[1 1 1] On Being Play
  (|CallParentFunction)
  (Transformation|SetWorldScale3D (Variables|Default|GetStaticMesh) (Variables|Default|GetMeshScale)))

(event Building|EventOnBuildingActorInitialized
  (params (InitializationReason :type EFortBuildingInitializationReason%20Enum :default "StaticallyPlaced") (BuildingPersistentState :type EFortBuildingPersistentState%20Enum :default "Default"))
  (bind _row_lifespan (Variables|Balance|GetRow_Lifespan))
  (bind _returnvalue (ScalableFloat|GetValueatLevel _row_lifespan))
  ;[1 1 1] On Building Actor Initialized
  (|CallParentFunction InitializationReason BuildingPersistentState)
  (Variables|Default|SetSlicesRemaining (Math|Integer|Clamp(Integer) (Math|Float|Truncate (ScalableFloat|GetValueatLevel (Variables|Balance|GetRow_MaxSliceCount))) 1 (Variables|Default|GetMaxPossibleSlices)))
  (Variables|Default|SetSetBoxOpen true)
  (Networking|SwitchHasAuthority
    (:Authority
      (CallFunction|MulticastMark))
    (:Remote
      _))
  (if (> _returnvalue 0.0)
    (Utilities|Time|SetTimerbyEvent (delegate Custom|KillSelf) _returnvalue)))

(event Custom|KillSelf
  ;[1 1 1] On Building Actor Initialized
  (Gameplay|Die self (Variables|Default|GetTag_KillSelf) 0 self))

(event Custom|SpawnSlicesOnDestroy
  ;[1 1 1] SpawnSlicesOnDestroy - Spawning Pizza Slices in case the Box is destroyed with pizza inside
  (if (>= (Variables|Default|GetSlicesRemaining) 1)
    (bind _numberofslices (CallFunction|GetNumberOfSlicesonDestroy)) ; Gets the amount of slices it should spawn
    (if (>= _numberofslices 1)
      (for _ (range 1 (+ _numberofslices 1))
        (Inventory|SpawnPickupInWorld (Variables|Default|GetPizzaSlice) 1 (Transformation|GetActorLocation) "0.000000,0.000000,1.000000" 0 true true false -1 "Tossed" "Unset" 0 false)))))

(event EventOnDeathServer
  (params (Damage :type Float%20%28single-precision%29) (DamageTags :type Gameplay%20Tag%20Container%20Structure%20%28by%20ref%29) (Momentum :type Vector :default "0, 0, 0") (HitInfo :type Hit%20Result%20Structure%20%28by%20ref%29) (InstigatedBy :type Controller%20Object%20Reference) (DamageCauser :type Actor%20Object%20Reference) (EffectContext :type Gameplay%20Effect%20Context%20Handle%20Structure))
  (bind _returnvalue_location (Math|Transform|BreakTransform (Transformation|GetActorTransform self)))
  ;[1 1 1] On Death Server
  (|CallParentFunction Damage DamageTags Momentum HitInfo InstigatedBy DamageCauser EffectContext)
  (Rendering|SetActorHiddenInGame self true)
  (Ability|ExecuteGameplayCueWithParamsReplicated (Variables|Default|GetTag_GC_Destroyed) (Ability|GameplayCue|MakeGameplayCueParameters :NormalizedMagnitude 0.0 :RawMagnitude 0.0 :Location _returnvalue_location :Normal "0,0,1.000000" :Instigator (Ability|EffectContext|GetInstigatorActor) :EffectCauser self :SourceObject self :GameplayEffectLevel 1 :AbilityLevel 1 :bReplicateLocationWhenUsingMinimalRepProxy false))
  (if (GameplayTags|HasTag DamageTags (Variables|Default|GetTag_KillSelf) true)
    (CallFunction|SpawnSlicesonDestroy) ; Spawning Pizza Slices
  ))

(event Interaction|EventBlueprintOnInteract
  (params (InteractingPawn :type Fort%20Pawn%20Object%20Reference) (InteractionBeingAttempted :type EInteractionBeingAttempted%20Enum :default "FirstInteraction"))
  (bind _returnvalue (Fortnite|GetFortPlayerControllerfromActor InteractingPawn))
  ;[1 1 1] On Interact
  (|CallParentFunction InteractingPawn InteractionBeingAttempted)
  (if (CallFunction|IsValidtoInteract InteractingPawn (Utilities|Enum|Equal(Enum) InteractionBeingAttempted "FirstInteraction"))
    (bind _result (Math|Integer|DecrementInt (Variables|Default|GetSlicesRemaining))
      (:%20%20
        (Variables|Default|SetSlicesRemaining _result)
        (switch Utilities|FlowControl|Switch|SwitchonEInteractionBeingAttempted InteractionBeingAttempted
          (:FirstInteraction
            ;[1 0 0] TODO: Use a child of GA_Athena_MedConsumable_Parent for the consume (get the consume timer wheel, the channel effect, make sure to drop held object, prevent the "cannot do this now" error when holding to interact, etc.)
            (GameplayEffects|ApplyGameplayEffectToSelf :self (Ability|GetAbilitySystemComponent InteractingPawn) :GameplayEffectClass (Variables|Default|GetGE_Consume) :Level 0.0) ; GE Triggers Consume Animation and Heal
            (CallFunction|CheckSlicesandDeleteifEmpty))
          (:SecondInteraction
            (if (Utilities|IsValid _returnvalue :!class /Script/Engine.KismetSystemLibrary)
              (CallFunction|MulticastPickUp)
              (Inventory|GiveItemtoInventoryOwner _returnvalue (Variables|Default|GetPizzaSlice))
              (CallFunction|CheckSlicesandDeleteifEmpty)
              (else
                (Development|PrintString "Invalid Fort Player Controller")
                (CallFunction|CheckSlicesandDeleteifEmpty)))))))))

(event Custom|Multicast_PickUp
  ;[1 1 1] Pickup Audio for Slice
  (Audio|SpawnSoundAttached "/FlipperGameplay/Items/PizzaParty/Audio/PizzaParty_TakeSlice_Cue.PizzaParty_TakeSlice_Cue" (Variables|Default|GetStaticMesh)))

(event Custom|CheckSlicesAndDeleteIfEmpty
  ;[1 1 1] Check Slices and kill self if empty
  (if (<= (Variables|Default|GetSlicesRemaining) 0)
    (CallFunction|KillSelf)))

(event EventMulticast_PlayHitFX
  (params (HitResult :type Hit%20Result%20Structure))
  ;[1 1 1] Bounce Land FX
  (Ability|ExecuteGameplayCueWithParamsReplicated (Variables|Default|GetTag_GC_HitWorld) (Ability|GameplayCue|MakeGameplayCueParameters 0.0)))

(event OnEnterWaterEvent(FortWaterInteraction)
  (params (WaterBody :type Fort%20Water%20Body%20Actor%20Object%20Reference) (WaterInteractionComponent :type Fort%20Water%20Interaction%20Component%20Object%20Reference) (bIsFirstBody :type Boolean))
  (bind _returnvalue_location (Math|Transform|BreakTransform (Transformation|GetActorTransform self)))
  ;[1 1 1] Water FX
  (if bIsFirstBody
    (Utilities|FlowControl|DoOnce false
      (:Completed
        (Ability|ExecuteGameplayCueWithParamsReplicated (Variables|Default|GetTag_GC_HitWater) (Ability|GameplayCue|MakeGameplayCueParameters :NormalizedMagnitude 0.0 :RawMagnitude 0.0 :Location _returnvalue_location :Normal "0,0,1.000000" :Instigator (Ability|EffectContext|GetInstigatorActor) :EffectCauser self :SourceObject self :GameplayEffectLevel 1 :AbilityLevel 1 :bReplicateLocationWhenUsingMinimalRepProxy false))
        (Utilities|Time|SetTimerbyEvent (EventDispatchers|CreateEvent :!event Custom|ResetEnterWaterEvent) (ScalableFloat|GetValueatLevel (Variables|Default|GetRow_WaterSplashFXCooldown))))) ; Timed gate to prevent FX spam in super wavy waters
    (else
      (Development|PrintString "Yort Enter Water"))))

(event Custom|ResetEnterWaterEvent
  (bind _returnvalue_location (Math|Transform|BreakTransform (Transformation|GetActorTransform self)))
  ;[1 1 1] Water FX
  (Utilities|FlowControl|DoOnce false
    (:Completed
      (Ability|ExecuteGameplayCueWithParamsReplicated (Variables|Default|GetTag_GC_HitWater) (Ability|GameplayCue|MakeGameplayCueParameters :NormalizedMagnitude 0.0 :RawMagnitude 0.0 :Location _returnvalue_location :Normal "0,0,1.000000" :Instigator (Ability|EffectContext|GetInstigatorActor) :EffectCauser self :SourceObject self :GameplayEffectLevel 1 :AbilityLevel 1 :bReplicateLocationWhenUsingMinimalRepProxy false))
      (Utilities|Time|SetTimerbyEvent (EventDispatchers|CreateEvent :!event Custom|ResetEnterWaterEvent) (ScalableFloat|GetValueatLevel (Variables|Default|GetRow_WaterSplashFXCooldown))))) ; Timed gate to prevent FX spam in super wavy waters
)

(event Custom|Multicast_Mark
  ;[1 1 1] Mark Item for Squad
  (bind _asfort_player_pawn_athena (Utilities|Casting|CastToFortPlayerPawnAthena (Actor|GetOwner)))
  (bind _asfort_player_controller_athena (Utilities|Casting|CastToFortPlayerControllerAthena (Pawn|GetController _asfort_player_pawn_athena)))
  (Fortnite|RegisterSoundIndicatorPlacementContext _asfort_player_controller_athena self))

(event Collision|EventHit
  (params (MyComp :type Primitive%20Component%20Object%20Reference) (Other :type Actor%20Object%20Reference) (OtherComp :type Primitive%20Component%20Object%20Reference) (bSelfMoved :type Boolean) (HitLocation :type Vector :default "0, 0, 0") (HitNormal :type Vector :default "0, 0, 0") (NormalImpulse :type Vector :default "0, 0, 0") (Hit :type Hit%20Result%20Structure%20%28by%20ref%29))
  (bind (_hitdata_mycomp_2_873970e843903ef1de91179fa7bb4ff5 _hitdata_other_7_9b1252b44981f3c40dea08a7411af77b _hitdata_othercomp_6_8b45cb97430e81f8f857adb1a1326da3 _hitdata_selfmoved_17_819c96774839f59c9190e983a4b37ccd _hitdata_hitlocation_18_c618d4da42c9e07e26ca4ca38ec6b880 _hitdata_hitnormal_19_aca68ffd44775d4d332c04abaac2d3fa _hitdata_normalimpulse_20_7407ffbd475cb1ab27851cb4ff1965eb _hitdata_hit_21_b73c35de4e7dbc02c3db759f5a5b4692) (Utilities|Struct|BreakStructHitData (Variables|GetHitData_MyComp_2_873970E843903EF1DE91179FA7BB4FF5)))
  ;[1 1 1] On Hit
  (Variables|SetHitData_MyComp_2_873970E843903EF1DE91179FA7BB4FF5 (Utilities|Struct|MakeStructHitData MyComp Other OtherComp bSelfMoved HitLocation HitNormal NormalImpulse Hit))
  (Utilities|Casting|CastToFortAthenaVehicle Other
    (:then
      (Variables|Default|SetLastHitVehicle true))
    (:CastFailed
      (if (Water|IsinWater (Variables|Default|GetFortWaterInteraction))
        (Variables|/Script/Engine.ProjectileMovementComponent|SetFriction 0.75 (Variables|Default|GetProjectileMovement)))
      (Variables|Default|SetLastHitVehicle false)
      (|CallParentFunction _hitdata_mycomp_2_873970e843903ef1de91179fa7bb4ff5 _hitdata_other_7_9b1252b44981f3c40dea08a7411af77b _hitdata_othercomp_6_8b45cb97430e81f8f857adb1a1326da3 _hitdata_selfmoved_17_819c96774839f59c9190e983a4b37ccd _hitdata_hitlocation_18_c618d4da42c9e07e26ca4ca38ec6b880 _hitdata_hitnormal_19_aca68ffd44775d4d332c04abaac2d3fa _hitdata_normalimpulse_20_7407ffbd475cb1ab27851cb4ff1965eb _hitdata_hit_21_b73c35de4e7dbc02c3db759f5a5b4692))))

(event OnProjectileStop(ProjectileMovement)
  (params (ImpactResult :type Hit%20Result%20Structure%20%28by%20ref%29))
  (bind (_returnvalue_x _returnvalue_y _returnvalue_z) (Math|Vector|BreakVector (Transformation|GetActorLocation self)))
  (bind (_hitdata_mycomp_2_873970e843903ef1de91179fa7bb4ff5 _hitdata_other_7_9b1252b44981f3c40dea08a7411af77b _hitdata_othercomp_6_8b45cb97430e81f8f857adb1a1326da3 _hitdata_selfmoved_17_819c96774839f59c9190e983a4b37ccd _hitdata_hitlocation_18_c618d4da42c9e07e26ca4ca38ec6b880 _hitdata_hitnormal_19_aca68ffd44775d4d332c04abaac2d3fa _hitdata_normalimpulse_20_7407ffbd475cb1ab27851cb4ff1965eb _hitdata_hit_21_b73c35de4e7dbc02c3db759f5a5b4692) (Utilities|Struct|BreakStructHitData (Variables|GetHitData_MyComp_2_873970E843903EF1DE91179FA7BB4FF5)))
  ;[1 1 1] On Stop
  (if (Variables|Default|GetLastHitVehicle)
    (Variables|Default|SetLastHitVehicle false)
    (if (ScalableFloat|GetValueasBool (Variables|Default|GetRow_KillOnHitVehicle))
      (CallFunction|KillSelf)
      (else
        (Transformation|SetActorLocation (Math|Vector|MakeVector _returnvalue_x _returnvalue_y (- _returnvalue_z 2.0)) false true)
        ;[1 1 1] On Hit
        (|CallParentFunction _hitdata_mycomp_2_873970e843903ef1de91179fa7bb4ff5 _hitdata_other_7_9b1252b44981f3c40dea08a7411af77b _hitdata_othercomp_6_8b45cb97430e81f8f857adb1a1326da3 _hitdata_selfmoved_17_819c96774839f59c9190e983a4b37ccd _hitdata_hitlocation_18_c618d4da42c9e07e26ca4ca38ec6b880 _hitdata_hitnormal_19_aca68ffd44775d4d332c04abaac2d3fa _hitdata_normalimpulse_20_7407ffbd475cb1ab27851cb4ff1965eb _hitdata_hit_21_b73c35de4e7dbc02c3db759f5a5b4692)))))

(event EventSetLifeSpanForMovingPlatformAttachedBuildingActors
  (params (LifeSpan :type Float%20%28single-precision%29))
  (|CallParentFunction (ScalableFloat|GetValueatLevel (Variables|Default|GetAttachmentTime))))

(fn BlueprintCanInteract
  (params (InteractingPawn :type Fort%20Pawn%20Object%20Reference) (InteractionBeingAttempted :type EInteractionBeingAttempted%20Enum :default "FirstInteraction") (InteractionType :type TInteractionType%20Enum :default "IT_NoInteraction"))
  (switch Utilities|FlowControl|Switch|SwitchonEInteractionBeingAttempted InteractionBeingAttempted
    (:FirstInteraction
      (return (CallFunction|IsValidtoInteract InteractingPawn true)))
    (:SecondInteraction
      (return (CallFunction|IsValidtoInteract InteractingPawn)))))

(fn BlueprintGetFailedInteractionString
  (params (InteractingPawn :type Fort%20Pawn%20Object%20Reference) (InteractionBeingAttempted :type EInteractionBeingAttempted%20Enum :default "FirstInteraction"))
  (if (and (Utilities|IsValid InteractingPawn :!class /Script/Engine.KismetSystemLibrary) (Utilities|IsValid (Ability|GetFortAbilitySystemComponent InteractingPawn) :!class /Script/Engine.KismetSystemLibrary))
    (if (Fortnite|CanPerformNativeAction (Ability|GetFortAbilitySystemComponent (Variables|GetInteractingPawn)) (Variables|Default|GetTag_InteractConsumableActionCheck))
      (switch Utilities|FlowControl|Switch|SwitchonEInteractionBeingAttempted (Variables|GetInteractionBeingAttempted)
        (:FirstInteraction
          (if (not (CallFunction|IsValidtoInteract (Variables|GetInteractingPawn) true))
            (return (Variables|Default|GetText_CannotEatSlice))))
        (:SecondInteraction
          _))
      (return (Variables|Default|GetText_CannotTakeSlice))
      (else
        (return "")))
    (else
      (return ""))))

(fn BlueprintGetInteractionString
  (params (InteractingPawn :type Fort%20Pawn%20Object%20Reference) (InteractionBeingAttempted :type EInteractionBeingAttempted%20Enum :default "FirstInteraction"))
  (switch Utilities|FlowControl|Switch|SwitchonEInteractionBeingAttempted InteractionBeingAttempted
    (:FirstInteraction
      (if (CallFunction|IsValidtoInteract InteractingPawn true)
        (return (Variables|Default|GetText_InteractConsume))
        (else
          (return (Variables|Default|GetText_CannotEatSlice)))))
    (:SecondInteraction
      (if (CallFunction|IsValidtoInteract InteractingPawn)
        (return (Variables|Default|GetText_InteractTake))
        (else
          (return (Variables|Default|GetText_CannotTakeSlice)))))))

(fn BlueprintGetInteractionTime
  (params (InteractingPawn :type Fort%20Pawn%20Object%20Reference) (InteractionBeingAttempted :type EInteractionBeingAttempted%20Enum :default "FirstInteraction"))
  (return (Utilities|Select :Index InteractionBeingAttempted :FirstInteraction (ScalableFloat|GetValueatLevel (Variables|Balance|GetRow_InteractTime_Consume)) :SecondInteraction (ScalableFloat|GetValueatLevel (Variables|Balance|GetRow_InteractTime_Take)) :AllInteraction 0.0) true))

(fn GetNumberOfSlicesOnDestroy
  (bind _row_percentageofslicestospawn (Variables|Balance|GetRow_PercentageOfSlicesToSpawn))
  (bind _returnvalue (ScalableFloat|GetValueatLevel _row_percentageofslicestospawn))
  (bind _returnvalue_2 (Math|Float|Clamp(Float) _returnvalue))
  (bind _slicesremaining (Variables|Default|GetSlicesRemaining))
  (bind _returnvalue_3 (Math|Conversions|ToFloat(Integer) _slicesremaining))
  (bind _returnvalue_4 (* _returnvalue_3 _returnvalue_2))
  (if (ScalableFloat|GetValueasBool (Variables|Balance|GetRow_RoundUpSlices))
    (return (Math|Float|Round _returnvalue_4))
    (else
      (return (Math|Float|Truncate _returnvalue_4)))))

(fn IsValidToInteract
  (params (InteractingPawn :type Fort%20Pawn%20Object%20Reference) (CheckHealth :type Boolean))
  (bind _row_maxshieldtointeract (Variables|Balance|GetRow_MaxShieldToInteract))
  (bind _returnvalue (ScalableFloat|GetValueatLevel _row_maxshieldtointeract))
  (bind _row_maxshieldtointeract_2 (Variables|Balance|GetRow_MaxShieldToInteract))
  (bind _returnvalue_2 (ScalableFloat|GetValueatLevel _row_maxshieldtointeract_2))
  (bind _interactingpawn (Variables|GetInteractingPawn))
  (bind _returnvalue_3 (FortPawn|GetShield _interactingpawn))
  (bind _returnvalue_4 (Math|Float|Round _returnvalue_3))
  (bind _returnvalue_5 (Math|Conversions|ToFloat(Integer) _returnvalue_4))
  (bind _returnvalue_6 (> _returnvalue_2 0.0))
  (bind _returnvalue_7 (FortPawn|GetMaxShield _interactingpawn))
  (bind _returnvalue_8 (* _returnvalue_7 _returnvalue))
  (bind _returnvalue_9 (< _returnvalue_5 _returnvalue_8))
  (if CheckHealth
    (if (> (ScalableFloat|GetValueatLevel (Variables|Balance|GetRow_MaxHealthToInteract)) 0.0)
      (LocalVariables|SetPassedHealthCheck (< (Math|Conversions|ToFloat(Integer) (Math|Float|Round (FortPawn|GetHealth (Variables|GetInteractingPawn)))) (* (FortPawn|GetMaxHealth (Variables|GetInteractingPawn)) (ScalableFloat|GetValueatLevel (Variables|Balance|GetRow_MaxHealthToInteract)))))
      (if _returnvalue_6
        (LocalVariables|SetPassedHealthCheck (or (LocalVariables|GetPassedHealthCheck) _returnvalue_9))
        (else
          _))
      (else
        (if _returnvalue_6
          (LocalVariables|SetPassedHealthCheck _returnvalue_9)
          (else
            (LocalVariables|SetPassedHealthCheck true)))))
    (else
      (LocalVariables|SetPassedHealthCheck true)))
  (return (Math|Boolean|ANDBoolean (Fortnite|CanPerformNativeAction (Ability|GetFortAbilitySystemComponent (Variables|GetInteractingPawn)) (Variables|Default|GetTag_InteractConsumableActionCheck)) (> (Variables|Default|GetSlicesRemaining) 0) (LocalVariables|GetPassedHealthCheck))))

(fn LocalOnFailedInteract
  (params (InteractingPawn :type Fort%20Player%20Pawn%20Object%20Reference))
  (|CallParentFunction InteractingPawn)
  (FailedAction|UnabletoPerformAction InteractingPawn true (Variables|Default|GetTag_FailReason_HealthFull)))

(fn OnRep_InteractsRemaining)

(fn OnRep_SetBoxOpen
  (if (Variables|Default|GetSetBoxOpen)
    (CallFunction|OpenBox)
    (else
      (CallFunction|CloseBox))))

(fn OnRep_SlicesRemaining
  (LocalVariables|SetSlicesMissing 0)
  (while (< (+ (Variables|Default|GetSlicesRemaining) (LocalVariables|GetSlicesMissing)) (Variables|Default|GetMaxPossibleSlices))
    (Rendering|Material|SetScalarParameterValueonMaterials (Variables|Default|GetStaticMesh) (Utilities|Name|MakeLiteralName (Utilities|String|StringToName (Utilities|String|Append "Hide Element 0" (Utilities|String|ToString(Integer) (- (Variables|Default|GetMaxPossibleSlices) (LocalVariables|GetSlicesMissing)))))) (* (- (Variables|Default|GetMaxPossibleSlices) (LocalVariables|GetSlicesMissing)) 0.1))
    (Math|Integer|IncrementInt (LocalVariables|GetSlicesMissing))))

(fn ConstructionScript
  (bind (_markerdisplay_displayname _markerdisplay_icon _markerdisplay_customindicatorclass _markerdisplay_customdetailsclass _markerdisplay_primarycolor _markerdisplay_sound _markerdisplay_battachsound _markerdisplay_screenclamping) (Utilities|Struct|BreakMarkedActorDisplayInfo (Variables|GetMarkerDisplay_DisplayName)))
  (|CallParentFunction)
  ;[1 1 1] Get the proper rarity color for the marker display
  (bind _returnvalue (Fortnite|GetRarityData "Legendary"))
  (Variables|SetMarkerDisplay_DisplayName (Utilities|Struct|MakeMarkedActorDisplayInfo _markerdisplay_displayname _markerdisplay_icon _markerdisplay_customindicatorclass _markerdisplay_customdetailsclass (.Color2 (Utilities|Struct|BreakFortRarityItemData _returnvalue)) _markerdisplay_sound true _markerdisplay_screenclamping)))
